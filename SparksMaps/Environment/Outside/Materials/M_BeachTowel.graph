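# Two-sided, Nanite-enabled lit material for beach towels in the outdoor environment, built to pick a towel design from a texture atlas.
# Base texture UVs are scaled by U/V Size, rotated by CustomRotator, shrunk by Atlas UV Size, then offset by UPos and VPos in cell-size steps.
# A separately tiled normal map has its red and green channels scaled by Normal Intensity.
# Clamped, rounded-up UPos and VPos branches do not feed the atlas lookup.
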Material /SparksMaps/Environment/Outside/Materials/M_BeachTowel
graph {
 properties {bUsedWithNanite=true, pixelDepthOffsetMode=PDOM_Legacy, twoSided=true, usedShadingModels=Default Lit}
 nodes [
  n1 = TextureSampleParameter2D(UVs=TextureCoordinate() {uTiling=2, unMirrorU=true, unMirrorV=true, vTiling=2}) {parameterName=Normal Texture, samplerType=SAMPLERTYPE_Normal, texture=/SparksMaps/Environment/Outside/Textures/T_BeachTowels_N}
  n2 = ScalarParameter() {defaultValue=0.8, parameterName=Normal Intensity}
  AppendVector(A=AppendVector(A=Multiply(A=@n1.R, B=@n2), B=Multiply(A=@n1.G, B=@n2)), B=@n1.B)
  n3 = ScalarParameter() {parameterName=VPos}
  Ceil(None=Clamp(None=@n3) {maxDefault=8})
  n4 = ScalarParameter() {parameterName=UPos}
  Ceil(None=Clamp(None=@n4) {maxDefault=8})
  Constant()
  n5 = ScalarParameter() {defaultValue=1, parameterName=Atlas UV Size, sliderMax=1}
  n6 = Multiply(A=@n4, B=@n5)
  n7 = Multiply(A=@n3, B=@n5)
  n8 = ScalarParameter() {defaultValue=1, parameterName=U Size}
  n9 = ScalarParameter() {defaultValue=1, parameterName=V Size}
  n10 = MaterialFunctionCall(UVs=Multiply(A=TextureCoordinate(), B=AppendVector(A=@n8, B=@n9))) {materialFunction=&CustomRotator}
  TextureSampleParameter2D(UVs=Add(A=AppendVector(A=@n6, B=@n7), B=Multiply(A=@n10.Rotated Values, B=@n5))) {parameterName=Base Texture, texture=/SparksMaps/Environment/Outside/Textures/T_T_GR_BEachTowels_A_D}
 ]
}
